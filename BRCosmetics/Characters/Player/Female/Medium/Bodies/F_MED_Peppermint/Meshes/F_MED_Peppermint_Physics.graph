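# Physics asset for the F_MED_Peppermint female player body, covering the costume's dangling pieces.
# Sleeve tail chains and ruffles hang from the forearms, bow loops, ribbons and a three-link cravat from bow_master_jnt, and puffballs from the calf twist bones.
# These joints have tight swing and twist limits, while bow_master_jnt is rotation-locked to spine_05 and forearm-to-spine links swing freely.
# Bodies are capsules, apart from boxes on the bow ribbons.
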PhysicsAsset /BRCosmetics/Characters/Player/Female/Medium/Bodies/F_MED_Peppermint/Meshes/F_MED_Peppermint_Physics
bodies {
 bow_master_jnt [{shapeType=Capsule, length=4, radius=1, shapeName=bow_master_jnt_capsule}]
 calf_twist_01_l [{shapeType=Capsule, length=1, radius=1, shapeName=calf_twist_01_l_capsule}]
 calf_twist_01_r [{shapeType=Capsule, length=1, radius=1, shapeName=calf_twist_01_r_capsule}]
 dyn_L_bow_ribbon_jnt [{shapeType=Box, extentX=3, extentY=0.5, extentZ=3.5, shapeName=dyn_L_bow_ribbon_jnt_box, center=<0.5829 0 0.3358>}]
 dyn_L_puffball_jnt [{shapeType=Capsule, length=1, radius=4, shapeName=dyn_L_puffball_jnt_capsule, center=<4.0856 0 0>}]
 dyn_L_sleeve_ruffles_back_jnt [{shapeType=Capsule, length=2, radius=2, shapeName=dyn_L_sleeve_ruffles_back_jnt_capsule}]
 dyn_L_sleeve_ruffles_front_jnt [{shapeType=Capsule, length=1.25, radius=0.75, shapeName=dyn_L_sleeve_ruffles_front_jnt_capsule}]
 dyn_L_sleeve_tail_lateral_1_jnt [{
   shapeType=Capsule
   length=7
   radius=2.75
   shapeName=dyn_L_sleeve_tail_lateral_1_jnt_capsule
   center=<8.8345 0.0007 0>
   rotation=<rot -90 -14.8511 14.8512>
  }]
 dyn_L_sleeve_tail_lateral_2_jnt [{
   shapeType=Capsule
   length=1.5
   radius=2.75
   shapeName=dyn_L_sleeve_tail_lateral_2_jnt_capsule
   center=<6.913 0.0008 0>
   rotation=<rot -90 -1.6876 1.6879>
  }]
 dyn_L_sleeve_tail_lateral_3_jnt [{
   shapeType=Capsule
   length=1.5
   radius=2.75
   shapeName=dyn_L_sleeve_tail_lateral_3_jnt_capsule
   center=<6.913 0.0007 0>
   rotation=<rot -90 1.2658 -1.2657>
  }]
 dyn_L_sleeve_tail_medial_1_jnt [{
   shapeType=Capsule
   length=1.5
   radius=2.75
   shapeName=dyn_L_sleeve_tail_medial_1_jnt_capsule
   center=<6.913 0.0007 0>
   rotation=<rot -90 -1.8984 1.8984>
  }]
 dyn_L_sleeve_tail_medial_2_jnt [{
   shapeType=Capsule
   length=1.5
   radius=2.75
   shapeName=dyn_L_sleeve_tail_medial_2_jnt_capsule
   center=<6.913 0.0008 0>
   rotation=<rot -90 -0.2373 0.2375>
  }]
 dyn_L_sleeve_tail_medial_3_jnt [{
   shapeType=Capsule
   length=1.5
   radius=2.75
   shapeName=dyn_L_sleeve_tail_medial_3_jnt_capsule
   center=<6.913 0.0007 0>
   rotation=<rot -90 -0.633 0.6331>
  }]
 dyn_L_sleeve_tail_medial_4_jnt [{
   shapeType=Capsule
   length=1.5
   radius=2.75
   shapeName=dyn_L_sleeve_tail_medial_4_jnt_capsule
   center=<6.913 0.0007 0>
   rotation=<rot -90 0 0.0001>
  }]
 dyn_R_bow_ribbon_jnt [{shapeType=Box, extentX=3, extentY=0.5, extentZ=3.5, shapeName=dyn_R_bow_ribbon_jnt_box, center=<-1.476 0 0>}]
 dyn_R_puffball_jnt [{shapeType=Capsule, length=1, radius=4, shapeName=dyn_R_puffball_jnt_capsule, center=<-4.3186 0 0>}]
 dyn_R_sleeve_ruffles_back_jnt [{shapeType=Capsule, length=2, radius=2, shapeName=dyn_R_sleeve_ruffles_back_jnt_capsule}]
 dyn_R_sleeve_ruffles_front_jnt [{shapeType=Capsule, length=2, radius=2, shapeName=dyn_L_sleeve_ruffles_front_jnt_capsule}]
 dyn_R_sleeve_tail_lateral_1_jnt [{
   shapeType=Capsule
   length=6.5
   radius=2.75
   shapeName=dyn_R_sleeve_tail_lateral_1_jnt_capsule
   center=<-8.774 0.0001 -1.5471>
   rotation=<rot -79.9998 0 0>
  }]
 dyn_R_sleeve_tail_lateral_2_jnt [{
   shapeType=Capsule
   length=1.5
   radius=2.75
   shapeName=dyn_R_sleeve_tail_lateral_2_jnt_capsule
   center=<-5.3341 0.0002 -0.9406>
   rotation=<rot -80 0 0>
  }]
 dyn_R_sleeve_tail_lateral_3_jnt [{
   shapeType=Capsule
   length=1.5
   radius=2.75
   shapeName=dyn_R_sleeve_tail_lateral_3_jnt_capsule
   center=<-5.3341 0.0002 -0.9406>
   rotation=<rot -79.9998 0.0001 0>
  }]
 dyn_R_sleeve_tail_medial_1_jnt [{
   shapeType=Capsule
   length=1.5
   radius=2.75
   shapeName=dyn_R_sleeve_tail_medial_1_jnt_capsule
   center=<-5.3341 0.0002 -0.9406>
   rotation=<rot -79.9997 -0.0001 0>
  }]
 dyn_R_sleeve_tail_medial_2_jnt [{
   shapeType=Capsule
   length=1.5
   radius=2.75
   shapeName=dyn_R_sleeve_tail_medial_2_jnt_capsule
   center=<-5.3341 0.0002 -0.9406>
   rotation=<rot -79.9998 0 0>
  }]
 dyn_R_sleeve_tail_medial_3_jnt [{
   shapeType=Capsule
   length=1.5
   radius=2.75
   shapeName=dyn_R_sleeve_tail_medial_3_jnt_capsule
   center=<-5.3341 0.0002 -0.9406>
   rotation=<rot -79.9998 0.0001 0>
  }]
 dyn_R_sleeve_tail_medial_4_jnt [{
   shapeType=Capsule
   length=1.5
   radius=2.75
   shapeName=dyn_R_sleeve_tail_medial_4_jnt_capsule
   center=<-5.3341 0.0002 -0.9406>
   rotation=<rot -80 0 0>
  }]
 dyn_bow_L_jnt [{shapeType=Capsule, length=3, radius=1.5, shapeName=dyn_bow_L_jnt_capsule, center=<0 0.0001 3.3837>}]
 dyn_bow_R_jnt [{shapeType=Capsule, length=3, radius=1.5, shapeName=dyn_bow_R_jnt_capsule, center=<0 -0.0001 -3.4633>}]
 dyn_bow_cravat_1_jnt [{shapeType=Capsule, length=4.75, radius=1.75, shapeName=dyn_bow_cravat_1_jnt_capsule, center=<1.7605 0.1533 -0.3661>}]
 dyn_bow_cravat_2_jnt [{shapeType=Capsule, length=5.5, radius=2.5, shapeName=dyn_bow_cravat_2_jnt_capsule, center=<2.3949 0.1446 0>}]
 dyn_bow_cravat_3_jnt [{shapeType=Capsule, length=5.5, radius=2.5, shapeName=dyn_bow_cravat_3_jnt_capsule, center=<2.4494 0.25 0>}]
 lowerarm_l [{shapeType=Capsule, length=1, radius=1, shapeName=lowerarm_l_capsule}]
 lowerarm_r [{shapeType=Capsule, length=1, radius=1, shapeName=lowerarm_r_capsule}]
 spine_04 [{shapeType=Capsule, length=14, radius=9.5, shapeName=spine_04_capsule, center=<-1.6701 2.6149 0>}]
 spine_05 [{shapeType=Capsule, length=14, radius=9, shapeName=spine_05_capsule, center=<-4.8598 3.1336 0>}]
}
constraints [
 {
  bone1Name=dyn_L_sleeve_tail_medial_2_jnt
  bone2Name=dyn_L_sleeve_tail_medial_1_jnt
  swing1LimitDegrees=15
  swing2LimitDegrees=15
  twistLimitDegrees=1
 }
 {
  bone1Name=dyn_L_sleeve_tail_medial_3_jnt
  bone2Name=dyn_L_sleeve_tail_medial_2_jnt
  swing1LimitDegrees=15
  swing2LimitDegrees=15
  twistLimitDegrees=1
 }
 {
  bone1Name=dyn_L_sleeve_tail_medial_4_jnt
  bone2Name=dyn_L_sleeve_tail_medial_3_jnt
  swing1LimitDegrees=15
  swing2LimitDegrees=15
  twistLimitDegrees=1
 }
 {
  bone1Name=dyn_L_sleeve_tail_lateral_2_jnt
  bone2Name=dyn_L_sleeve_tail_lateral_1_jnt
  swing1LimitDegrees=15
  swing2LimitDegrees=15
  twistLimitDegrees=1
 }
 {
  bone1Name=dyn_L_sleeve_tail_lateral_3_jnt
  bone2Name=dyn_L_sleeve_tail_lateral_2_jnt
  swing1LimitDegrees=15
  swing2LimitDegrees=15
  twistLimitDegrees=1
 }
 {
  bone1Name=dyn_R_sleeve_tail_medial_2_jnt
  bone2Name=dyn_R_sleeve_tail_medial_1_jnt
  swing1LimitDegrees=15
  swing2LimitDegrees=15
  twistLimitDegrees=1
 }
 {
  bone1Name=dyn_R_sleeve_tail_medial_3_jnt
  bone2Name=dyn_R_sleeve_tail_medial_2_jnt
  swing1LimitDegrees=15
  swing2LimitDegrees=15
  twistLimitDegrees=1
 }
 {
  bone1Name=dyn_R_sleeve_tail_medial_4_jnt
  bone2Name=dyn_R_sleeve_tail_medial_3_jnt
  swing1LimitDegrees=15
  swing2LimitDegrees=15
  twistLimitDegrees=1
 }
 {
  bone1Name=dyn_R_sleeve_tail_lateral_2_jnt
  bone2Name=dyn_R_sleeve_tail_lateral_1_jnt
  swing1LimitDegrees=15
  swing2LimitDegrees=15
  twistLimitDegrees=1
 }
 {
  bone1Name=dyn_R_sleeve_tail_lateral_3_jnt
  bone2Name=dyn_R_sleeve_tail_lateral_2_jnt
  swing1LimitDegrees=15
  swing2LimitDegrees=15
  twistLimitDegrees=1
 }
 {bone1Name=dyn_bow_cravat_1_jnt, bone2Name=bow_master_jnt, swing1LimitDegrees=15, swing2LimitDegrees=15, twistLimitDegrees=2}
 {bone1Name=dyn_bow_cravat_2_jnt, bone2Name=dyn_bow_cravat_1_jnt, swing1LimitDegrees=15, swing2LimitDegrees=15, twistLimitDegrees=2}
 {bone1Name=dyn_bow_cravat_3_jnt, bone2Name=dyn_bow_cravat_2_jnt, swing1LimitDegrees=15, swing2LimitDegrees=15, twistLimitDegrees=2}
 {bone1Name=dyn_bow_R_jnt, bone2Name=bow_master_jnt, swing1LimitDegrees=15, swing2LimitDegrees=15, twistLimitDegrees=1}
 {bone1Name=dyn_R_bow_ribbon_jnt, bone2Name=dyn_bow_R_jnt, swing1LimitDegrees=5, swing2LimitDegrees=5, twistLimitDegrees=1}
 {bone1Name=dyn_bow_L_jnt, bone2Name=bow_master_jnt, swing1LimitDegrees=15, swing2LimitDegrees=15, twistLimitDegrees=1}
 {bone1Name=dyn_L_bow_ribbon_jnt, bone2Name=dyn_bow_L_jnt, swing1LimitDegrees=5, swing2LimitDegrees=5, twistLimitDegrees=1}
 {bone1Name=dyn_L_sleeve_tail_medial_1_jnt, bone2Name=lowerarm_l, swing1LimitDegrees=15, swing2LimitDegrees=15, twistLimitDegrees=1}
 {bone1Name=dyn_L_sleeve_ruffles_front_jnt, bone2Name=lowerarm_l, swing1LimitDegrees=15, swing2LimitDegrees=15, twistLimitDegrees=1}
 {bone1Name=dyn_L_sleeve_tail_lateral_1_jnt, bone2Name=lowerarm_l, swing1LimitDegrees=15, swing2LimitDegrees=15, twistLimitDegrees=1}
 {bone1Name=dyn_L_sleeve_ruffles_back_jnt, bone2Name=lowerarm_l, swing1LimitDegrees=10, swing2LimitDegrees=10, twistLimitDegrees=1}
 {bone1Name=dyn_R_sleeve_ruffles_back_jnt, bone2Name=lowerarm_r, swing1LimitDegrees=10, swing2LimitDegrees=10, twistLimitDegrees=1}
 {bone1Name=dyn_R_sleeve_ruffles_front_jnt, bone2Name=lowerarm_r, swing1LimitDegrees=10, swing2LimitDegrees=10, twistLimitDegrees=1}
 {bone1Name=dyn_R_sleeve_tail_medial_1_jnt, bone2Name=lowerarm_r, swing1LimitDegrees=15, swing2LimitDegrees=15, twistLimitDegrees=1}
 {bone1Name=dyn_R_sleeve_tail_lateral_1_jnt, bone2Name=lowerarm_r, swing1LimitDegrees=15, swing2LimitDegrees=15, twistLimitDegrees=1}
 {
  bone1Name=lowerarm_l
  bone2Name=spine_05
  swing1LimitDegrees=45
  swing1Motion=Free
  swing2LimitDegrees=45
  swing2Motion=Free
  twistLimitDegrees=45
  twistMotion=Free
 }
 {
  bone1Name=lowerarm_r
  bone2Name=spine_05
  swing1LimitDegrees=45
  swing1Motion=Free
  swing2LimitDegrees=45
  swing2Motion=Free
  twistLimitDegrees=45
  twistMotion=Free
 }
 {
  bone1Name=bow_master_jnt
  bone2Name=spine_05
  swing1LimitDegrees=45
  swing1Motion=Locked
  swing2LimitDegrees=45
  swing2Motion=Locked
  twistLimitDegrees=45
  twistMotion=Locked
 }
 {bone1Name=dyn_L_puffball_jnt, bone2Name=calf_twist_01_l, swing1LimitDegrees=15, swing2LimitDegrees=15, twistLimitDegrees=2}
 {bone1Name=dyn_R_puffball_jnt, bone2Name=calf_twist_01_r, swing1LimitDegrees=15, swing2LimitDegrees=15, twistLimitDegrees=2}
 {
  bone1Name=spine_05
  bone2Name=spine_04
  swing1LimitDegrees=45
  swing1Motion=Free
  swing2LimitDegrees=45
  swing2Motion=Free
  twistLimitDegrees=45
  twistMotion=Free
 }
]
