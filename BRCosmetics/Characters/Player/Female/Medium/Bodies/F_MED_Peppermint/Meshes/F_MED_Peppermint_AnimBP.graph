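# Post-process style animation Blueprint for the F_MED_Peppermint female character body, adding physics-driven secondary motion.
# It copies the pose from the attached parent mesh, then chains two base-bone-space RigidBody simulations: a skirt physics asset, then the main Peppermint physics asset with damped component motion.
# The event graph computes a thigh alpha from thigh_l yaw and a backbling flag on customization load, though the AnimGraph reads neither.
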AnimBlueprint /BRCosmetics/Characters/Player/Female/Medium/Bodies/F_MED_Peppermint/Meshes/F_MED_Peppermint_AnimBP
target_skeleton=&/F_MED_Peppermint_Skeleton
graphs [
 AnimGraph {nodes [
   n1 = RigidBody(ComponentPose=Local To Component(LocalPose=Copy Pose From Mesh() {bUseAttachedParent=true})) {
    bOverrideWorldGravity=true
    overridePhysicsAsset=&/F_MED_Peppermint_Skirt_Physics
    simulationSpace=BaseBoneSpace
    alphaBoolBlend {blendOption=Linear}
    baseBoneRef {boneName=root}
    componentAppliedLinearAccClamp=<1000 1000 1000>
    componentLinearAccScale=<1 1 1>
    componentLinearVelScale=<1 1 1>
    overrideWorldGravity {z=-980}
   }
   n2 = RigidBody(ComponentPose=@n1) {
    bClampLinearTranslationLimitToRefPose=true
    bOverrideWorldGravity=true
    overridePhysicsAsset=&/F_MED_Peppermint_Physics
    simulationSpace=BaseBoneSpace
    alphaBoolBlend {blendOption=Linear}
    baseBoneRef {boneName=root}
    componentAppliedLinearAccClamp=<3000 3000 3000>
    componentLinearAccScale=<0.3 0.2 0.2>
    componentLinearVelScale=<0.3 0.2 0.2>
    overrideWorldGravity {z=-980}
   }
   Output Pose(Result=Component To Local(ComponentPose=@n2)) {layerGroup=None}
  ]}
 EventGraph {nodes [
   Event BlueprintUpdateAnimation()
   n1 = BreakTransform(InTransform=GetDeltaTransformFromRefPose(self=GetOwnerSkeletalMeshComponent(), BoneName=thigh_l, BaseName=pelvis))
   > Set R thighAlpha(R thighAlpha:double=MapRangeClamped(Value=BreakRotator(InRot=@n1.Rotation).Yaw, InRangeA=-39, InRangeB=-80)) -> (Output_Get:double)
   Event OnCharacterCustomizationLoaded()
   n2 = CheckForBackbling()
   > Set Is Wearing Backbling(Is Wearing Backbling=OR Boolean(A=@n2, B=@n2.bIsBackblingLong))
  ]}
]
variables [
 Is Wearing Backbling {type=boolean}
 R thighAlpha {type=number}
]
