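# Control Rig for Peppermint's face accessory mesh that adjusts the back hair chain from head rotation.
# Each run sets the initial local transform of helper bone spine05_AIMTGT, then offsets it by a weight remapped and clamped from the head's local Z Euler angle.
# AimItem then points dyn_hair_back_1_jnt at that helper, using neck_01 as secondary target.
# An unconnected head-to-neck relative transform node is unused.
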ControlRigBlueprint /BRCosmetics/Characters/Player/Female/Medium/Bodies/F_MED_Peppermint/Meshes/Parts/F_MED_Peppermint_FaceAcc_CtrlRig
graphs [
 RigVMFunctionLibrary {links [], nodes []}
 RigVMModel {nodes [
   RigUnit_BeginExecution = &RigUnit_BeginExecution()
   &RigUnit_SetTransform(
    Item=(Type=Bone,Name="spine05_AIMTGT")
    Space=LocalSpace
    Value=(Rotation=(X=0,Y=0,Z=0,W=1),Translation=(X=0,Y=-9,Z=0),Scale3D=(X=1,Y=1,Z=1))
    Weight=1
    bInitial=true
    ExecutePin=@RigUnit_BeginExecution.ExecutePin
   )
   n1 = &RigUnit_GetTransform(Item=(Type=Bone,Name="head"), Space=LocalSpace)
   > &RigUnit_ModifyTransforms(
    ItemToModify=((Item=(Type=Bone,Name="spine05_AIMTGT"),Transform=(Rotation=(X=0,Y=0,Z=0,W=1),Translation=(X=0,Y=-3,Z=0),Scale3D=(X=1,Y=1,Z=1))))
    Weight=&RigVMFunction_MathFloatRemap(
     SourceMaximum=-35
     SourceMinimum=-10
     bClamp=true
     Value=&RigVMFunction_MathQuaternionToEuler(RotationOrder=ZYX, Value=@n1.Transform.Rotation).Result.Z
    ).Result
   )
   > /Script/ControlRig.RigUnit_AimItem(
    DebugSettings=(Scale=10,WorldOffset=(Rotation=(X=0,Y=0,Z=0,W=1),Translation=(X=0,Y=0,Z=0),Scale3D=(X=1,Y=1,Z=1)))
    Item=(Type=Bone,Name="dyn_hair_back_1_jnt")
    Primary=(Weight=1,Axis=(X=1,Y=0,Z=0),Target=(X=1,Y=0,Z=0),Kind=Location,Space=(Type=Bone))
    Secondary=(Weight=1,Axis=(X=0,Y=1,Z=0),Target=(X=1,Y=0,Z=0),Kind=Location,Space=(Type=Bone))
    Weight=1
    Primary.Target=&RigUnit_GetTransform(Item=(Type=Bone,Name="spine05_AIMTGT")).Transform.Translation
    Secondary.Target=&RigUnit_GetTransform(Item=(Type=Bone,Name="neck_01")).Transform.Translation
   ) {is=struct RigVMUnitNode}
   &RigUnit_GetRelativeTransformForItem(Child=(Type=Bone,Name="head"), Parent=(Type=Bone,Name="neck_01"), bParentInitial=true)
  ]}
]
hierarchy [
 root {loc=<0 0 0>, rot=<0 0 0 1>}
 pelvis {parent=root, loc=<0 -0.5301 98.0571>, rot=<0.0127 0.707 -0.0127 -0.707>}
 spine_01 {parent=pelvis, loc=<6.1208 -0.9328 0>, rot=<0 0 0.034 -0.9994>}
 spine_02 {parent=spine_01, loc=<6.5996 -0.2095 0>, rot=<0 0 0.0238 -0.9997>}
 spine_03 {parent=spine_02, loc=<6.4914 -0.1237 0>, rot=<0 0 0.0149 -0.9999>}
 spine_04 {parent=spine_03, loc=<6.9598 0.0794 0>, rot=<0 0 -0.0132 -0.9999>}
 spine_05 {parent=spine_04, loc=<6.9987 0 0>, rot=<0 0 -0.0015 -1>}
 clavicle_l {parent=spine_05, loc=<5.4831 -0.1053 -3.2946>, rot=<0.0465 -0.7289 -0.009 -0.6829>}
 clavicle_r {parent=spine_05, loc=<5.4831 -0.1053 3.2946>, rot=<0.7289 0.0465 -0.6829 0.009>}
 neck_01 {parent=spine_05, loc=<11.3138 0.1529 0>, rot=<0 0 -0.2003 -0.9797>}
 neck_02 {parent=neck_01, loc=<4.5627 0 0>, rot=<0 0 0.0504 -0.9987>}
 head {parent=neck_02, loc=<5.3914 0 0>, rot=<0 0 0.115 -0.9934>}
 hair_Master_jnt {parent=head, loc=<3.7712 0 0>, rot=<0 0 0 1>}
 dyn_L_hair_side_jnt {parent=hair_Master_jnt, loc=<-4.862 1.9327 -6.5808>, rot=<-0.131 -0.9896 0.0224 0.0553>}
 dyn_R_hair_side_jnt {parent=hair_Master_jnt, loc=<-4.8624 1.9327 6.5809>, rot=<0.9896 -0.131 0.0553 -0.0224>}
 dyn_hair_back_1_jnt {parent=hair_Master_jnt, loc=<4.21 -4.9343 0.9203>, rot=<0.1086 -0.994 0.001 -0.0102>}
 dyn_hair_back_2_jnt {parent=dyn_hair_back_1_jnt, loc=<4.7249 0.0758 0>, rot=<0 0 0 1>}
 dyn_hair_back_3_jnt {parent=dyn_hair_back_2_jnt, loc=<17.094 -1.4643 0>, rot=<0 0 0 1>}
 dyn_hair_back_4_jnt {parent=dyn_hair_back_3_jnt, loc=<14.57 -3.2424 0.3665>, rot=<0 0 0 1>}
 dyn_bangs_L_jnt {parent=hair_Master_jnt, loc=<6.1542 9.4812 -3.1415>, rot=<0.0029 -0.7071 -0.0029 -0.7071>}
 dyn_bangs_R_jnt {parent=hair_Master_jnt, loc=<6.1522 9.7326 2.5923>, rot=<0.0029 -0.7071 -0.0029 -0.7071>}
 spine05_AIMTGT {parent=spine_05, loc=<0 -9 0>, rot=<0 0 0 1>}
]
