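# Physics asset for the F_MED_Peppermint female medium player cosmetic's face accessory part, set up for dynamic hair.
# Side-hair boxes, bang capsules and a four-joint back-hair chain hang from hair_Master_jnt through angular constraints.
# The chain's root swings and twists freely, while its later links have limited swing and very tight twist.
# Head, neck, clavicle, spine and pelvis capsules have no constraints, apparently serving as collision volumes for the hair.
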PhysicsAsset /BRCosmetics/Characters/Player/Female/Medium/Bodies/F_MED_Peppermint/Meshes/Parts/F_MED_Peppermint_FaceAcc_Physics
bodies {
 clavicle_l [{shapeType=Capsule, length=5.9448, radius=6.9569, center=<7.9237 -1.944 0.7223>, rotation=<rot 90 0 -170>}]
 clavicle_r [{shapeType=Capsule, length=5.6948, radius=7.2069, center=<-9.4156 1.9439 -0.1196>, rotation=<rot 90 0 0>}]
 dyn_L_hair_side_jnt [{shapeType=Box, extentX=6, extentY=7.5, extentZ=1, shapeName=dyn_L_hair_side_jnt_box, center=<2.9519 0 0>, rotation=<rot -10 0 0>}]
 dyn_R_hair_side_jnt [{shapeType=Box, extentX=6, extentY=7.5, extentZ=1, shapeName=dyn_R_hair_side_jnt_box, center=<-2.867 0 0.0001>, rotation=<rot -10 0 0>}]
 dyn_bangs_L_jnt [{shapeType=Capsule, length=2, radius=1.5, shapeName=dyn_bangs_L_jnt_capsule, center=<-0.2951 0.2581 -2.5679>}]
 dyn_bangs_R_jnt [{shapeType=Capsule, length=2, radius=1.5, shapeName=dyn_bangs_R_jnt_capsule, center=<0.6268 0.2581 -2.5679>}]
 dyn_hair_back_1_jnt [{shapeType=Capsule, length=0.505, radius=0.505, shapeName=dyn_hair_back_1_jnt_capsule, rotation=<rot 0 -90 0>}]
 dyn_hair_back_2_jnt [{shapeType=Capsule, length=5.5, radius=7.5, shapeName=dyn_hair_back_2_jnt_capsule, center=<7 -4 0>}]
 dyn_hair_back_3_jnt [{shapeType=Capsule, length=14, radius=8.5, shapeName=dyn_hair_back_3_jnt_capsule, center=<5 -7.5 0>}]
 dyn_hair_back_4_jnt [{shapeType=Capsule, length=14.75, radius=14.5, shapeName=dyn_hair_back_4_jnt_capsule, center=<8 -14 3>}]
 hair_Master_jnt [{shapeType=Capsule, length=1, radius=1, shapeName=hair_Master_jnt_capsule}]
 head [{shapeType=Capsule, length=6, radius=7, center=<3.5101 2.2324 0>, rotation=<rot 90 0 0>}]
 neck_01 [{shapeType=Capsule, length=4.5, radius=4.25, center=<1.5642 0 0>, rotation=<rot 90 0 0>}]
 pelvis [{shapeType=Capsule, length=9.25, radius=12.75, center=<-0.8809 1.5936 0>}]
 spine_02 [{shapeType=Capsule, length=9.75, radius=8.25, center=<-0.8665 2.1785 0.0001>}]
 spine_03 [{shapeType=Capsule, length=9.5, radius=9.25, center=<4.9633 3.5337 0>}]
 spine_05 [{shapeType=Capsule, length=37.5, radius=8.75, center=<3.3588 -0.7472 0.0001>}]
}
constraints [
 {bone1Name=dyn_L_hair_side_jnt, bone2Name=hair_Master_jnt, swing1LimitDegrees=50, swing2LimitDegrees=50, twistLimitDegrees=5}
 {bone1Name=dyn_R_hair_side_jnt, bone2Name=hair_Master_jnt, swing1LimitDegrees=50, swing2LimitDegrees=50, twistLimitDegrees=5}
 {
  bone1Name=dyn_hair_back_1_jnt
  bone2Name=hair_Master_jnt
  swing1LimitDegrees=45
  swing1Motion=Free
  swing2LimitDegrees=45
  swing2Motion=Free
  twistLimitDegrees=45
  twistMotion=Free
 }
 {bone1Name=dyn_bangs_L_jnt, bone2Name=hair_Master_jnt, swing1LimitDegrees=10, swing2LimitDegrees=10, twistLimitDegrees=10}
 {bone1Name=dyn_bangs_R_jnt, bone2Name=hair_Master_jnt, swing1LimitDegrees=10, swing2LimitDegrees=10, twistLimitDegrees=10}
 {bone1Name=dyn_hair_back_2_jnt, bone2Name=dyn_hair_back_1_jnt, swing1LimitDegrees=45, swing2LimitDegrees=45, twistLimitDegrees=2}
 {bone1Name=dyn_hair_back_3_jnt, bone2Name=dyn_hair_back_2_jnt, swing1LimitDegrees=45, swing2LimitDegrees=45, twistLimitDegrees=2}
 {bone1Name=dyn_hair_back_4_jnt, bone2Name=dyn_hair_back_3_jnt, swing1LimitDegrees=20, swing2LimitDegrees=20, twistLimitDegrees=2}
]
