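# Animation Blueprint for the F_MED_Peppermint face accessory mesh, posing its back-hair bones around back bling, capes and targeting.
# The parent mesh pose runs through a Control Rig and head-space RigidBody simulation into a cached pose.
# A back-bling switch picks between two additive bone-offset chains with interpolated targeting alphas and a front-end tweak, then capes scale down two hair bones.
# CheckForBackbling refreshes these flags on every update and on customization load.
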AnimBlueprint /BRCosmetics/Characters/Player/Female/Medium/Bodies/F_MED_Peppermint/Meshes/Parts/F_MED_Peppermint_FaceAcc_AnimBP
target_skeleton=&/F_MED_Peppermint_FaceAcc_Skeleton
graphs [
 AnimGraph {nodes [
   n1 = "Transform (Modify) Bone - Bone: dyn_hair_back_1_jnt"(ComponentPose=Local To Component(LocalPose=Use cached pose 'cache pose'()), Alpha=1, Translation="0,0,0", Rotation="0,-5,10", Scale="1,1,1") {
    rotationMode=BMM_Additive
    rotationSpace=BCS_BoneSpace
    alphaBoolBlend {blendOption=Linear}
    boneToModify {boneName=dyn_hair_back_1_jnt}
    rotation {roll=10, yaw=-5}
   }
   n2 = "Transform (Modify) Bone - Bone: dyn_hair_back_3_jnt"(ComponentPose=@n1, Alpha=1, Translation="0,0,0", Rotation="0,-10,0", Scale="1,1,1") {
    rotationMode=BMM_Additive
    rotationSpace=BCS_BoneSpace
    alphaBoolBlend {blendOption=Linear}
    boneToModify {boneName=dyn_hair_back_3_jnt}
    rotation {yaw=-10}
   }
   n3 = "Transform (Modify) Bone - Bone: dyn_hair_back_4_jnt"(ComponentPose=@n2, Alpha=1, Translation="0,0,0", Rotation="0,-10,0", Scale="1,1,1") {
    rotationMode=BMM_Additive
    rotationSpace=BCS_BoneSpace
    alphaBoolBlend {blendOption=Linear}
    boneToModify {boneName=dyn_hair_back_4_jnt}
    rotation {yaw=-10}
   }
   n4 = "Transform (Modify) Bone - Bone: dyn_hair_back_1_jnt"(ComponentPose=@n3, Alpha="To Float (Boolean)"(InBool=Get bIsTargeting()), Translation="0,-3,0", Rotation="0,10,0", Scale="1,1,1") {
    rotationMode=BMM_Additive
    rotationSpace=BCS_BoneSpace
    translationMode=BMM_Additive
    translationSpace=BCS_BoneSpace
    alphaBoolBlend {blendOption=Linear}
    alphaScaleBiasClamp {bInterpResult=true, interpSpeedDecreasing=8}
    boneToModify {boneName=dyn_hair_back_1_jnt}
    rotation {yaw=10}
    translation {y=-3}
   }
   n5 = "Transform (Modify) Bone - Bone: dyn_hair_back_1_jnt"!(ComponentPose=@n4, bAlphaBoolEnabled:bool=True, Translation="0,0,0", Rotation="0,5,0", Scale="1,1,1") -> (Pose) {
    alphaInputType=Bool
    rotationMode=BMM_Additive
    rotationSpace=BCS_BoneSpace
    alphaBoolBlend {blendOption=Linear}
    boneToModify {boneName=dyn_hair_back_1_jnt}
    rotation {yaw=5}
   }
   n6 = "Transform (Modify) Bone - Bone: dyn_hair_back_3_jnt"!(ComponentPose=@n5, bAlphaBoolEnabled:bool=True, Translation="0,0,0", Rotation="0,22,0", Scale="1,1,1") -> (Pose) {
    alphaInputType=Bool
    rotationMode=BMM_Additive
    rotationSpace=BCS_BoneSpace
    translationMode=BMM_Additive
    translationSpace=BCS_BoneSpace
    alphaBoolBlend {blendOption=Linear}
    boneToModify {boneName=dyn_hair_back_3_jnt}
    rotation {yaw=22}
   }
   n7 = "Transform (Modify) Bone - Bone: dyn_hair_back_4_jnt"!(ComponentPose=@n6, bAlphaBoolEnabled:bool=True, Translation="0,0,0", Rotation="0,15,0", Scale="1,1,1") -> (Pose) {
    alphaInputType=Bool
    rotationMode=BMM_Additive
    rotationSpace=BCS_BoneSpace
    translationMode=BMM_Additive
    translationSpace=BCS_BoneSpace
    alphaBoolBlend {blendOption=Linear}
    boneToModify {boneName=dyn_hair_back_4_jnt}
    rotation {yaw=15}
   }
   n8 = "To Float (Boolean)"(InBool=Get bIsTargeting())
   n9 = "Transform (Modify) Bone - Bone: dyn_hair_back_1_jnt"!(ComponentPose=@n7, Alpha:float=@n8, Translation="0,-4,0", Rotation="0,0,20", Scale="1,1,1") -> (Pose) {
    rotationMode=BMM_Additive
    rotationSpace=BCS_BoneSpace
    translationMode=BMM_Additive
    translationSpace=BCS_BoneSpace
    alphaBoolBlend {blendOption=Linear}
    alphaScaleBiasClamp {bInterpResult=true, interpSpeedDecreasing=8}
    boneToModify {boneName=dyn_hair_back_1_jnt}
    rotation {roll=20}
    translation {y=-4}
   }
   n10 = "Transform (Modify) Bone - Bone: dyn_hair_back_3_jnt"!(ComponentPose=@n9, Alpha:float=@n8, Translation="0,1,0", Rotation="0,0,-20", Scale="1,1,1") -> (Pose) {
    rotationMode=BMM_Additive
    rotationSpace=BCS_BoneSpace
    translationMode=BMM_Additive
    translationSpace=BCS_BoneSpace
    alphaBoolBlend {blendOption=Linear}
    alphaScaleBiasClamp {bInterpResult=true, interpSpeedDecreasing=2, interpSpeedIncreasing=5}
    boneToModify {boneName=dyn_hair_back_3_jnt}
    rotation {roll=-20}
    translation {y=1}
   }
   n11 = "Transform (Modify) Bone - Bone: dyn_hair_back_4_jnt"!(ComponentPose=@n10, Alpha:float=@n8, Translation="0,0,0", Rotation="0,10,15", Scale="1,1,1") -> (Pose) {
    rotationMode=BMM_Additive
    rotationSpace=BCS_BoneSpace
    translationMode=BMM_Additive
    translationSpace=BCS_BoneSpace
    alphaBoolBlend {blendOption=Linear}
    alphaScaleBiasClamp {bInterpResult=true, interpSpeedDecreasing=2, interpSpeedIncreasing=5}
    boneToModify {boneName=dyn_hair_back_4_jnt}
    rotation {roll=15, yaw=10}
   }
   n12 = Component To Local(ComponentPose=@n11)
   n13 = "Transform (Modify) Bone - Bone: dyn_hair_back_1_jnt"(ComponentPose=Local To Component(LocalPose=Use cached pose 'cache pose'()), Alpha=1, Translation="0,0,0", Rotation="0,-5,10", Scale="1,1,1") {
    rotationMode=BMM_Additive
    rotationSpace=BCS_BoneSpace
    alphaBoolBlend {blendOption=Linear}
    boneToModify {boneName=dyn_hair_back_1_jnt}
    rotation {roll=10, yaw=-5}
   }
   n14 = "Transform (Modify) Bone - Bone: dyn_hair_back_3_jnt"(ComponentPose=@n13, Alpha=1, Translation="0,0,0", Rotation="0,-10,0", Scale="1,1,1") {
    rotationMode=BMM_Additive
    rotationSpace=BCS_BoneSpace
    alphaBoolBlend {blendOption=Linear}
    boneToModify {boneName=dyn_hair_back_3_jnt}
    rotation {yaw=-10}
   }
   n15 = "Transform (Modify) Bone - Bone: dyn_hair_back_4_jnt"(ComponentPose=@n14, Alpha=1, Translation="0,0,0", Rotation="0,-10,0", Scale="1,1,1") {
    rotationMode=BMM_Additive
    rotationSpace=BCS_BoneSpace
    alphaBoolBlend {blendOption=Linear}
    boneToModify {boneName=dyn_hair_back_4_jnt}
    rotation {yaw=-10}
   }
   n16 = "Transform (Modify) Bone - Bone: dyn_hair_back_1_jnt"(ComponentPose=@n15, Alpha="To Float (Boolean)"(InBool=Get bIsTargeting()), Translation="0,-3,0", Rotation="0,10,0", Scale="1,1,1") {
    rotationMode=BMM_Additive
    rotationSpace=BCS_BoneSpace
    translationMode=BMM_Additive
    translationSpace=BCS_BoneSpace
    alphaBoolBlend {blendOption=Linear}
    alphaScaleBiasClamp {bInterpResult=true, interpSpeedDecreasing=8}
    boneToModify {boneName=dyn_hair_back_1_jnt}
    rotation {yaw=10}
    translation {y=-3}
   }
   n17 = Get Is Wearing Backbling()
   n18 = "Transform (Modify) Bone - Bone: dyn_hair_back_1_jnt"!(ComponentPose=@n16, bAlphaBoolEnabled:bool=@n17, Translation="0,0,0", Rotation="0,5,0", Scale="1,1,1") -> (Pose) {
    alphaInputType=Bool
    rotationMode=BMM_Additive
    rotationSpace=BCS_BoneSpace
    alphaBoolBlend {blendOption=Linear}
    boneToModify {boneName=dyn_hair_back_1_jnt}
    rotation {yaw=5}
   }
   n19 = "Transform (Modify) Bone - Bone: dyn_hair_back_3_jnt"!(ComponentPose=@n18, bAlphaBoolEnabled:bool=@n17, Translation="0,0,0", Rotation="0,22,0", Scale="1,1,1") -> (Pose) {
    alphaInputType=Bool
    rotationMode=BMM_Additive
    rotationSpace=BCS_BoneSpace
    translationMode=BMM_Additive
    translationSpace=BCS_BoneSpace
    alphaBoolBlend {blendOption=Linear}
    boneToModify {boneName=dyn_hair_back_3_jnt}
    rotation {yaw=22}
   }
   n20 = "Transform (Modify) Bone - Bone: dyn_hair_back_4_jnt"!(ComponentPose=@n19, bAlphaBoolEnabled:bool=@n17, Translation="0,0,0", Rotation="0,15,0", Scale="1,1,1") -> (Pose) {
    alphaInputType=Bool
    rotationMode=BMM_Additive
    rotationSpace=BCS_BoneSpace
    translationMode=BMM_Additive
    translationSpace=BCS_BoneSpace
    alphaBoolBlend {blendOption=Linear}
    boneToModify {boneName=dyn_hair_back_4_jnt}
    rotation {yaw=15}
   }
   n21 = "To Float (Boolean)"(InBool=AND Boolean(A=Get bIsTargeting(), B=Get Is Wearing Backbling()))
   n22 = "Transform (Modify) Bone - Bone: dyn_hair_back_1_jnt"!(ComponentPose=@n20, Alpha:float=@n21, Translation="0,-4,0", Rotation="0,0,20", Scale="1,1,1") -> (Pose) {
    rotationMode=BMM_Additive
    rotationSpace=BCS_BoneSpace
    translationMode=BMM_Additive
    translationSpace=BCS_BoneSpace
    alphaBoolBlend {blendOption=Linear}
    alphaScaleBiasClamp {bInterpResult=true, interpSpeedDecreasing=8}
    boneToModify {boneName=dyn_hair_back_1_jnt}
    rotation {roll=20}
    translation {y=-4}
   }
   n23 = "Transform (Modify) Bone - Bone: dyn_hair_back_3_jnt"!(ComponentPose=@n22, Alpha:float=@n21, Translation="0,1,0", Rotation="0,0,-20", Scale="1,1,1") -> (Pose) {
    rotationMode=BMM_Additive
    rotationSpace=BCS_BoneSpace
    translationMode=BMM_Additive
    translationSpace=BCS_BoneSpace
    alphaBoolBlend {blendOption=Linear}
    alphaScaleBiasClamp {bInterpResult=true, interpSpeedDecreasing=2, interpSpeedIncreasing=5}
    boneToModify {boneName=dyn_hair_back_3_jnt}
    rotation {roll=-20}
    translation {y=1}
   }
   n24 = "Transform (Modify) Bone - Bone: dyn_hair_back_4_jnt"!(ComponentPose=@n23, Alpha:float=@n21, Translation="0,0,0", Rotation="0,10,15", Scale="1,1,1") -> (Pose) {
    rotationMode=BMM_Additive
    rotationSpace=BCS_BoneSpace
    translationMode=BMM_Additive
    translationSpace=BCS_BoneSpace
    alphaBoolBlend {blendOption=Linear}
    alphaScaleBiasClamp {bInterpResult=true, interpSpeedDecreasing=2, interpSpeedIncreasing=5}
    boneToModify {boneName=dyn_hair_back_4_jnt}
    rotation {roll=15, yaw=10}
   }
   n25 = "Transform (Modify) Bone - Bone: dyn_hair_back_4_jnt"!(
    ComponentPose=@n24
    bAlphaBoolEnabled:bool=AND Boolean(A=Get bIsInFrontEnd(), B=Get Is Wearing Backbling())
    Translation="0,0,0"
    Rotation="0,20,0"
    Scale="1,1,1"
   ) -> (Pose) {
    alphaInputType=Bool
    rotationMode=BMM_Additive
    rotationSpace=BCS_BoneSpace
    translationSpace=BCS_BoneSpace
    alphaBoolBlend {blendOption=Linear}
    boneToModify {boneName=dyn_hair_back_4_jnt}
    rotation {yaw=20}
   }
   n26 = Component To Local(ComponentPose=@n25)
   n27 = Get Is Wearing Backbling()
   n28 = Get Is Wearing Cape() -> (Is Wearing Cape:bool)
   n29 = "Transform (Modify) Bone - Bone: dyn_hair_back_3_jnt"!(
    ComponentPose=Local To Component(LocalPose=Blend Poses by bool(BlendPose_0=@n26, BlendPose_1=@n12, bActiveValue=@n27) {blendType=Linear})
    bAlphaBoolEnabled:bool=@n28
    Translation="0,0,0"
    Rotation="0,0,0"
    Scale="0.7,0.7,0.7"
   ) -> (Pose) {
    alphaInputType=Bool
    rotationSpace=BCS_BoneSpace
    scaleMode=BMM_Additive
    scaleSpace=BCS_BoneSpace
    alphaBoolBlend {blendOption=Linear}
    boneToModify {boneName=dyn_hair_back_3_jnt}
    scale=<0.7 0.7 0.7>
   }
   n30 = "Transform (Modify) Bone - Bone: dyn_hair_back_4_jnt"(ComponentPose=@n29, bAlphaBoolEnabled=@n28, Translation="0,0,0", Rotation="0,0,0", Scale="0.3,0.3,0.3") {
    alphaInputType=Bool
    rotationSpace=BCS_BoneSpace
    scaleMode=BMM_Additive
    scaleSpace=BCS_BoneSpace
    alphaBoolBlend {blendOption=Linear}
    boneToModify {boneName=dyn_hair_back_4_jnt}
    scale=<0.3 0.3 0.3>
   }
   Output Pose(Result=Component To Local(ComponentPose=@n30)) {layerGroup=None}
   n31 = Control Rig(Source=Copy Pose From Mesh() {bUseAttachedParent=true}) {
    alphaBoolBlend {blendOption=Linear}
    controlRigAssetReference {blueprintRigClass=&/F_MED_Peppermint_FaceAcc_CtrlRig.F_MED_Peppermint_FaceAcc_CtrlRig_C}
   }
   n32 = RigidBody(ComponentPose=Local To Component(LocalPose=@n31), Alpha=SelectFloat(A=0.2, B=0.2, bPickA=Get Is Wearing Backbling())) {
    bClampLinearTranslationLimitToRefPose=true
    bOverrideWorldGravity=true
    overridePhysicsAsset=&/F_MED_Peppermint_FaceAcc_Physics
    simulationSpace=BaseBoneSpace
    alphaBoolBlend {blendOption=Linear}
    baseBoneRef {boneName=head}
    componentAppliedLinearAccClamp=<3000 3000 3000>
    componentLinearAccScale=<0.3 0.2 0.2>
    componentLinearVelScale=<0.3 0.2 0.2>
    overrideWorldGravity {z=-980}
   }
   Save cached pose 'cache pose'(Pose=Component To Local(ComponentPose=@n32)) {cacheName=cache pose}
  ]}
 EventGraph {nodes [
   n1 = Event BlueprintUpdateAnimation()
   n2 = CheckForBackbling()
   Set Is Wearing Backbling(execute=[@n1.then, @n3.then], Is Wearing Backbling=OR Boolean(A=@n2, B=@n2.bIsBackblingLong))
   > Set Is Wearing Cape(Is Wearing Cape=@n2.bIsBackblingACape)
   n3 = Event OnCharacterCustomizationLoaded()
  ]}
]
variables [
 Is Wearing Backbling {type=boolean}
 Is Wearing Cape {type=boolean}
]
